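# Translucent unlit Niagara sprite material for toon-style smoke, located with the cosmetic emote prop spawn effects.
# Particle color, scaled by a multiplier, feeds emissive adjustments, while a dynamic parameter drives flipbook phase and a Picker blend between red-channel and alpha masks.
# The mask times particle alpha gives opacity, depth-faded on default and deferred paths but not forward or mobile.
# A static switch can scale fade distance by the inverted mask.
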Material /Game/Effects/Niagara/Systems/CosmeticTemplates/Emote/PropSpawn/M_Default_PropSpawn_Smoke_Toon
graph {
 properties {
  bAutomaticallySetUsageInEditor=false
  bUsedWithNiagaraSprites=true
  blendMode=BLEND_Translucent
  shadingModel=MSM_Unlit
  usedShadingModels=Unlit
 }
 nodes [
  n1 = ParticleColor()
  MaterialFunctionCall(Emissive=Multiply(A=@n1.RGB, B=ScalarParameter() {defaultValue=0.3, parameterName=Color Multiply})) {materialFunction=&MF_FX_EmissiveAdjustments}
  n2 = DynamicParameter() {defaultValue=<rgba 1 0 1 1>, paramNames [Time,Picker,Param3,Param4]}
  n3 = MaterialFunctionCall(
   "Animation  Phase (0-1)"=@n2.Time
   Number of Rows=ScalarParameter() {defaultValue=3, parameterName=Flipbook Rows}
   Number of Columns=ScalarParameter() {defaultValue=3, parameterName=Flipbook Columns}
   Texture=TextureObjectParameter() {parameterName=FlipbookTexture, texture=&/T_Default_Smoke_Toon}
  ) {materialFunction=&FlipBook}
  n4 = MaterialFunctionCall(Float3=@n3.Result) {materialFunction=&BreakOutFloat3Components}
  n5 = MaterialFunctionCall!(Mask 1=@n4.R, Mask 2=@n4.R, Mask 3=@n4.R, Mask 4=@n3.Alpha, Blend=@n2.Picker) {materialFunction=/Game/Effects/Niagara/MasterMaterials/Functions/Utility/MF_Lerp_4_Scalar}
  n6 = Multiply(A=@n1.A, B=@n5.Result)
  n7 = ScalarParameter() {defaultValue=12, parameterName=DepthFadeDis}
  n8 = OneMinus(None=@n5.Result)
  n9 = DepthFade(Opacity=@n6, FadeDistance=StaticSwitchParameter(True=Multiply(A=@n7, B=@n8), False=@n7) {parameterName=TextureDepthFade})
  ShadingPathSwitch(Default=@n9, Deferred=@n9, Forward=@n6, Mobile=@n6)
 ]
}
